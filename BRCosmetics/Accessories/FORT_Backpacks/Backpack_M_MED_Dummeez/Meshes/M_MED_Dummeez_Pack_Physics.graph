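# Physics asset for the Dummeez back bling cosmetic, giving its handle, wheel and ropes physical bodies for secondary motion.
# Box bodies cover the pack and wheel, while capsules cover the handle and four rope bones.
# Wheel and handle attach to the pack with single-axis rotation, twist for the wheel and swing for the handle.
# The ropes chain from the wheel, with angle limits tightening after the first link.
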PhysicsAsset /BRCosmetics/Accessories/FORT_Backpacks/Backpack_M_MED_Dummeez/Meshes/M_MED_Dummeez_Pack_Physics
bodies {
 dyn_handle [{shapeType=Capsule, length=7, radius=3, center=<-1.5 0 -1.5>}]
 dyn_ropes_01 [{shapeType=Capsule, length=1, radius=1, shapeName=dyn_ropes_01_capsule}]
 dyn_ropes_02 [{shapeType=Capsule, length=1, radius=1, shapeName=dyn_ropes_02_capsule}]
 dyn_ropes_03 [{shapeType=Capsule, length=1, radius=1, shapeName=dyn_ropes_03_capsule}]
 dyn_ropes_04 [{shapeType=Capsule, length=1, radius=1, shapeName=dyn_ropes_04_capsule, center=<5 0 0>}]
 dyn_wheel [{shapeType=Box, extentX=4, extentY=23, extentZ=23}]
 pack [{shapeType=Box, extentX=18.5, extentY=7, extentZ=23, center=<0 -2.8104 -3.632>}]
}
constraints [
 {
  bone1Name=dyn_wheel
  bone2Name=pack
  swing1LimitDegrees=45
  swing1Motion=Locked
  swing2LimitDegrees=45
  swing2Motion=Locked
  twistLimitDegrees=60
 }
 {
  bone1Name=dyn_handle
  bone2Name=pack
  swing1LimitDegrees=3
  swing1Motion=Locked
  swing2LimitDegrees=20
  twistLimitDegrees=3
  twistMotion=Locked
 }
 {bone1Name=dyn_ropes_01, bone2Name=dyn_wheel, swing1LimitDegrees=45, swing2LimitDegrees=45, twistLimitDegrees=15}
 {bone1Name=dyn_ropes_02, bone2Name=dyn_ropes_01, swing1LimitDegrees=10, swing2LimitDegrees=10, twistLimitDegrees=1}
 {bone1Name=dyn_ropes_03, bone2Name=dyn_ropes_02, swing1LimitDegrees=5, swing2LimitDegrees=5, twistLimitDegrees=1}
 {bone1Name=dyn_ropes_04, bone2Name=dyn_ropes_03, swing1LimitDegrees=5, swing2LimitDegrees=5, twistLimitDegrees=1}
]
