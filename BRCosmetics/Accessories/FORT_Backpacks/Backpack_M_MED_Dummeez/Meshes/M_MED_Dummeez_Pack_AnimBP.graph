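# Animation Blueprint for the Dummeez backpack cosmetic that adds secondary physics motion to the pack.
# The input pose feeds a RigidBody node using the pack physics asset, simulated relative to the pack bone with damped component motion and custom gravity, then an additive root-bone Z offset.
# An Anim Dynamics spring node for dyn_handle sits unconnected, and the event graph holds only unwired default nodes.
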AnimBlueprint /BRCosmetics/Accessories/FORT_Backpacks/Backpack_M_MED_Dummeez/Meshes/M_MED_Dummeez_Pack_AnimBP
target_skeleton=&/M_MED_Dummeez_Pack_Skeleton
graphs [
 AnimGraph {nodes [
   Anim Dynamics - Bone: dyn_handle(Alpha=1, ExternalForce="0,0,0") {
    angularSpringConstant=2000
    bAngularSpring=true
    bGravityOverrideInSimSpace=true
    bUseGravityOverride=true
    alphaBoolBlend {blendOption=Linear}
    boundBone {boneName=dyn_handle}
    componentAppliedLinearAccClamp=<500 500 1000>
    componentLinearAccScale {z=0.2}
    componentLinearVelScale {z=0.2}
    gravityOverride {z=-980}
    physicsBodyDefinitions [{
      sphereCollisionRadius=0
      boundBone {boneName=dyn_handle}
      boxExtents=<10 8 8>
      constraintSetup {angularLimitsMax=<0 25 0>, angularLimitsMin=<0 0 0>, angularTarget=<0 0 0>}
      localJointOffset=<-1.5 0 -1.5>
     }]
   }
   n1 = RigidBody(ComponentPose=Local To Component(LocalPose=Input Pose())) {
    bClampLinearTranslationLimitToRefPose=true
    bOverrideWorldGravity=true
    overridePhysicsAsset=&/M_MED_Dummeez_Pack_Physics
    simulationSpace=BaseBoneSpace
    alphaBoolBlend {blendOption=Linear}
    baseBoneRef {boneName=pack}
    componentAppliedLinearAccClamp=<3000 3000 3000>
    componentLinearAccScale=<0.3 0.2 0.3>
    componentLinearVelScale=<0.3 0.2 0.3>
    overrideWorldGravity {z=-800}
   }
   n2 = "Transform (Modify) Bone - Bone: root"!(ComponentPose=@n1, Alpha, Translation="0,0,4") -> (Pose) {
    translationMode=BMM_Additive
    translationSpace=BCS_BoneSpace
    alphaBoolBlend {blendOption=Linear}
    boneToModify {boneName=root}
    translation {z=4}
   }
   Output Pose(Result=Component To Local(ComponentPose=@n2)) {layerGroup=None}
  ]}
 EventGraph {nodes [
   TryGetPawnOwner()
   Event BlueprintUpdateAnimation()
  ]}
]
variables []
